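# Physics asset for the Durrburger Hero backpack accessory, defining its simplified collision and joint setup.
# Two capsule bodies cover the backpack, one on the backpack_root bone and a somewhat smaller one on drv_trans.
# A single constraint links drv_trans to backpack_root, with equal swing and twist limit angles.
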PhysicsAsset /Game/Accessories/FORT_Backpacks/M_MED_Durrburger_Hero_Backpack/Meshes/Backpack_Durrburger_Hero_PhysicsAsset
bodies {
 backpack_root [{shapeType=Capsule, length=22.518, radius=15.9146, center=<-1.9556 -6.1057 -0.5462>, rotation=<rot 0 -80.1521 -86.7815>}]
 drv_trans [{shapeType=Capsule, length=15.4889, radius=15.1489, center=<4.0753 -5.338 -0.9802>, rotation=<rot 10.7631 -90 -63.4585>}]
}
constraints [{bone1Name=drv_trans, bone2Name=backpack_root, swing1LimitDegrees=45, swing2LimitDegrees=45, twistLimitDegrees=45}]
